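# Unlit, two-sided background material in the front-end environment's Misc folder.
# Its graph holds a single Color vector parameter, defaulting to opaque black, with no connection to any material output shown.
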Material /Game/Environments/FrontEnd/Misc/M_UnlitBackground
graph {
 properties {pixelDepthOffsetMode=PDOM_Legacy, shadingModel=MSM_Unlit, twoSided=true, usedShadingModels=Unlit}
 nodes [VectorParameter() {parameterName=Color, defaultValue=<rgba 0 0 0 1>}]
}
