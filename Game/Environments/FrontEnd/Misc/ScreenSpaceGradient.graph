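# Unlit, translucent, two-sided front-end material that builds a screen-space radial gradient between two color parameters.
# A soft SphereMask on screen-aligned UVs, centered mid-screen with radius from Col 1's alpha, blends Col 1 into Col 2.
# The screen-resolution ratio scaling UVs and center divides height by itself, so it applies no aspect correction.
# The blend feeds MF_FX_EmissiveAdjustments, alongside a separate Opacity parameter.
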Material /Game/Environments/FrontEnd/Misc/ScreenSpaceGradient
graph {
 properties {
  bUseTranslucencyVertexFog=false
  blendMode=BLEND_Translucent
  pixelDepthOffsetMode=PDOM_Legacy
  shadingModel=MSM_Unlit
  twoSided=true
  usedShadingModels=Unlit
 }
 nodes [
  Constant()
  n1 = VectorParameter() {parameterName=Col 1, defaultValue=<rgba 0.0047 0.005 0.0046 0.5558>}
  n2 = MaterialFunctionCall(Float2=MaterialFunctionCall() {materialFunction=&ScreenResolution}.Visible Resolution) {materialFunction=&BreakOutFloat2Components}
  n3 = Divide(A=@n2.G, B=@n2.G)
  n4 = Multiply(
   A=AppendVector(A=Constant() {r=1}, B=@n3)
   B=MaterialFunctionCall() {materialFunction=&ScreenAlignedUVs}."X 100%, Y 100%"
  )
  n5 = Constant() {r=0.5}
  n6 = AppendVector(A=@n5, B=Multiply(A=@n5, B=@n3))
  MaterialFunctionCall(Emissive=LinearInterpolate(
    A=@n1.RGB
    B=VectorParameter() {parameterName=Col 2, defaultValue=<rgba 0.02 0.02 0.02 1>}.RGB
    Alpha=SphereMask(A=@n4, B=@n6, Radius=@n1.A) {attenuationRadius=0.75, hardnessPercent=0}
   )) {materialFunction=&MF_FX_EmissiveAdjustments}
  ScalarParameter() {defaultValue=1, parameterName=Opacity}
 ]
}
